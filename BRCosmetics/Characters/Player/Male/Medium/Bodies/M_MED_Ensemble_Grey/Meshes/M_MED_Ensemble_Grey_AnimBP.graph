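# Post-process style cosmetic AnimBP for the M_MED_Ensemble_Grey player body, adding sleeve and collar correctives to the pose copied from its parent mesh.
# The copied pose runs through a deform Control Rig, then an additive front right sleeve offset applies while targeting.
# The collar is constrained to spine_05 and clavicle_r, then rotated by an alpha the EventGraph derives from head-versus-spine_05 rotation, zeroed when not targeting.
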AnimBlueprint /BRCosmetics/Characters/Player/Male/Medium/Bodies/M_MED_Ensemble_Grey/Meshes/M_MED_Ensemble_Grey_AnimBP
target_skeleton=&/M_MED_Ensemble_Grey_Skeleton
graphs [
 AnimGraph {nodes [
   n1 = Control Rig(Source=Copy Pose From Mesh() {bUseAttachedParent=true}) {
    alphaBoolBlend {blendOption=Linear}
    controlRigAssetReference {blueprintRigClass=&/M_MED_Ensemble_Grey_DeformRig_Master_ControlRig.M_MED_Ensemble_Grey_DeformRig_Master_ControlRig_C}
   }
   n2 = "Transform (Modify) Bone - Bone: dyn_sleeveFront_r_01"!(
    ComponentPose=Local To Component(LocalPose=@n1)
    bAlphaBoolEnabled:bool=Get bIsTargeting()
    Translation="-2,0,0"
    Rotation="-0.000,25.096,36.244"
    Scale="1,1,1"
   ) -> (Pose) {
    alphaInputType=Bool
    rotationMode=BMM_Additive
    translationMode=BMM_Additive
    alphaBoolBlend {blendOption=Linear}
    boneToModify {boneName=dyn_sleeveFront_r_01}
    rotation=<rot 0 25.096 36.244>
    translation {x=-2}
   }
   n3 = "Constraint - dyn_collar_01 (PP)"!(ComponentPose=@n2, Alpha=1, ConstraintWeights_0=1, ConstraintWeights_1:float=1) -> (Pose) {
    alphaBoolBlend {blendOption=Linear}
    boneToModify {boneName=dyn_collar_01}
    constraintSetup [{targetBone {boneName=spine_05}}, {targetBone {boneName=clavicle_r}}]
    constraintWeights [1,1]
   }
   n4 = "Transform (Modify) Bone - Bone: dyn_collar_01"(ComponentPose=@n3, Alpha=Get head Rot Up() -> (head Rot Up:double), Translation="0,0,0", Rotation="-30,0,0", Scale="1,1,1") {
    rotationMode=BMM_Additive
    rotationSpace=BCS_BoneSpace
    translationMode=BMM_Additive
    alphaBoolBlend {blendOption=Linear}
    boneToModify {boneName=dyn_collar_01}
    rotation {pitch=-30}
   }
   Output Pose(Result=Component To Local(ComponentPose=@n4)) {layerGroup=None}
  ]}
 EventGraph {nodes [
   Event BlueprintUpdateAnimation()
   > n1 = Sequence()
   > Set Body Mesh(Body Mesh:Skeletal Mesh Component Object Reference=GetPartSkeletalMeshComponent(self=Self-Reference())) -> (Output_Get:Skeletal Mesh Component Object Reference)
   n2 = Branch(execute=@n1.then_1, Condition=Get bIsTargeting())
   > Set head Rot Up(head Rot Up:double=GetAlphaFromReferenceTransformDelta(
     SkeletalMeshComponent=Get Body Mesh() -> (Body Mesh:Skeletal Mesh Component Object Reference)
     BoneName=head
     BaseName=spine_05
     InputRange=(X=-40,Y=-20)
     OutputRange=(X=1,Y=0)
     TwistAxis="1,0,0"
    )) -> (Output_Get:double)
   Set head Rot Up(execute=@n2.else, head Rot Up=0)
  ]}
]
variables [
 Body Mesh {type=SkeletalMeshComponent}
 head Rot Up {type=number}
]
