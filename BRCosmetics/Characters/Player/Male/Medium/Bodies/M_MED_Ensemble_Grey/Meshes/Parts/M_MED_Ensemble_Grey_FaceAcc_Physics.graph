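# Physics asset for the FaceAcc part of the male medium Grey ensemble body, providing capsule bodies and joints for its dynamic hair bones.
# Front locks and a tail led by a MSTR bone form chains from the head, while all side bones constrain directly to the head.
# Hair joints have tight swing limits and zero twist limits.
# Head, neck and upper-spine capsules, linked by angularly free joints, give the hair body shapes to collide against.
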PhysicsAsset /BRCosmetics/Characters/Player/Male/Medium/Bodies/M_MED_Ensemble_Grey/Meshes/Parts/M_MED_Ensemble_Grey_FaceAcc_Physics
bodies {
 dyn_hairMeshFront_l_01 [{shapeType=Capsule, length=2.25, radius=1.75, shapeName=dyn_hairMeshFront_l_01_capsule, center=<2.2005 0 0>, rotation=<rot 90 0 0>}]
 dyn_hairMeshFront_l_02 [{shapeType=Capsule, length=3, radius=1.75, shapeName=dyn_hairMeshFront_l_02_capsule, center=<3.5489 0 0>, rotation=<rot 90 0 0>}]
 dyn_hairMeshFront_r_01 [{shapeType=Capsule, length=2.25, radius=1.75, shapeName=dyn_hairMeshFront_r_01_capsule, center=<2.2005 0 0>, rotation=<rot 90 0 0>}]
 dyn_hairMeshFront_r_02 [{shapeType=Capsule, length=3, radius=1.75, shapeName=dyn_hairMeshFront_r_02_capsule, center=<3.5489 0 0>, rotation=<rot 90 0 0>}]
 dyn_hairMeshSide_l_01 [{shapeType=Capsule, length=3.5, radius=3.5, shapeName=dyn_hairMeshSide_l_01_capsule, center=<3.588 0 2.5>, rotation=<rot 90 0 0>}]
 dyn_hairMeshSide_l_02 [{shapeType=Capsule, length=3.5, radius=3.5, shapeName=dyn_hairMeshSide_l_02_capsule, center=<3.588 0 1.9>, rotation=<rot 90 0 0>}]
 dyn_hairMeshSide_r_01 [{shapeType=Capsule, length=4.5, radius=3.75, shapeName=dyn_hairMeshSide_r_01_capsule, center=<4.6273 0 -1.8>, rotation=<rot 90 0 0>}]
 dyn_hairMeshSide_r_02 [{shapeType=Capsule, length=3.75, radius=3.25, shapeName=dyn_hairMeshSide_r_02_capsule, center=<3.588 0 -1.8>, rotation=<rot 90 0 0>}]
 dyn_hairTail_01 [{shapeType=Capsule, length=1, radius=4.75, shapeName=dyn_hairTail_01_capsule, center=<4.0332 0 3.8>, rotation=<rot 90 0 0>}]
 dyn_hairTail_02 [{shapeType=Capsule, length=5, radius=5, shapeName=dyn_hairTail_02_capsule, center=<3.7663 0 3.3>, rotation=<rot 78.6644 -0.0004 -0.0008>}]
 dyn_hairTail_03 [{shapeType=Capsule, length=5, radius=3, shapeName=dyn_hairTail_03_capsule, center=<3.2749 0 1.5>, rotation=<rot 90 0 0>}]
 dyn_hairTail_04 [{shapeType=Capsule, length=3.75, radius=2, shapeName=dyn_hairTail_04_capsule, center=<3.1686 0 0.25>, rotation=<rot 90 0 0>}]
 dyn_hairTail_MSTR_01 [{shapeType=Capsule, length=1, radius=1, shapeName=dyn_hairTail_MSTR_01_capsule}]
 head [
  {shapeType=Capsule, length=5.7486, radius=7.272, center=<9.9917 0.8191 0>, rotation=<rot 90 6.1155 -71.1389>}
  {shapeType=Capsule, length=7.5, radius=7.25, shapeName=head_capsule, rotation=<rot 90 0 0>}
 ]
 neck_01 [{shapeType=Capsule, length=1.75, radius=8.25, shapeName=neck_01_capsule, center=<0 2.3664 0>, rotation=<rot 0 -90 0>}]
 neck_02 [{shapeType=Capsule, length=0, radius=9.5, shapeName=neck_02_capsule, center=<-2.3315 2.9361 0>, rotation=<rot 0 -90 0>}]
 spine_04 [{shapeType=Capsule, length=13.75, radius=11.5, shapeName=spine_04_capsule, center=<9.2249 0.5999 0>, rotation=<rot 0 -90 0>}]
 spine_05 [{shapeType=Capsule, length=19.5, radius=10.75, shapeName=spine_05_capsule, center=<3.8469 1.7465 0>, rotation=<rot 0 -90 0>}]
}
constraints [
 {bone1Name=dyn_hairMeshFront_l_01, bone2Name=head, swing1LimitDegrees=20, swing2LimitDegrees=20, twistLimitDegrees=0}
 {bone1Name=dyn_hairMeshFront_l_02, bone2Name=dyn_hairMeshFront_l_01, swing1LimitDegrees=20, swing2LimitDegrees=20, twistLimitDegrees=0}
 {bone1Name=dyn_hairMeshSide_l_01, bone2Name=head, swing1LimitDegrees=30, swing2LimitDegrees=20, twistLimitDegrees=0}
 {bone1Name=dyn_hairMeshSide_l_02, bone2Name=head, swing1LimitDegrees=30, swing2LimitDegrees=20, twistLimitDegrees=0}
 {bone1Name=dyn_hairMeshSide_r_01, bone2Name=head, swing1LimitDegrees=30, swing2LimitDegrees=20, twistLimitDegrees=0}
 {bone1Name=dyn_hairMeshSide_r_02, bone2Name=head, swing1LimitDegrees=30, swing2LimitDegrees=20, twistLimitDegrees=0}
 {bone1Name=dyn_hairMeshFront_r_01, bone2Name=head, swing1LimitDegrees=20, swing2LimitDegrees=20, twistLimitDegrees=0}
 {bone1Name=dyn_hairMeshFront_r_02, bone2Name=dyn_hairMeshFront_r_01, swing1LimitDegrees=20, swing2LimitDegrees=20, twistLimitDegrees=0}
 {bone1Name=dyn_hairTail_MSTR_01, bone2Name=head, swing1LimitDegrees=20, swing2LimitDegrees=20, twistLimitDegrees=0}
 {bone1Name=dyn_hairTail_01, bone2Name=dyn_hairTail_MSTR_01, swing1LimitDegrees=10, swing2LimitDegrees=50, twistLimitDegrees=0}
 {bone1Name=dyn_hairTail_02, bone2Name=dyn_hairTail_01, swing1LimitDegrees=20, swing2LimitDegrees=20, twistLimitDegrees=0}
 {bone1Name=dyn_hairTail_03, bone2Name=dyn_hairTail_02, swing1LimitDegrees=20, swing2LimitDegrees=20, twistLimitDegrees=0}
 {bone1Name=dyn_hairTail_04, bone2Name=dyn_hairTail_03, swing1LimitDegrees=20, swing2LimitDegrees=20, twistLimitDegrees=0}
 {
  bone1Name=head
  bone2Name=neck_02
  swing1LimitDegrees=45
  swing1Motion=Free
  swing2LimitDegrees=45
  swing2Motion=Free
  twistLimitDegrees=45
  twistMotion=Free
 }
 {
  bone1Name=spine_05
  bone2Name=spine_04
  swing1LimitDegrees=45
  swing1Motion=Free
  swing2LimitDegrees=45
  swing2Motion=Free
  twistLimitDegrees=45
  twistMotion=Free
 }
 {
  bone1Name=neck_01
  bone2Name=spine_05
  swing1LimitDegrees=45
  swing1Motion=Free
  swing2LimitDegrees=45
  swing2Motion=Free
  twistLimitDegrees=45
  twistMotion=Free
 }
 {
  bone1Name=neck_02
  bone2Name=neck_01
  swing1LimitDegrees=45
  swing1Motion=Free
  swing2LimitDegrees=45
  swing2Motion=Free
  twistLimitDegrees=45
  twistMotion=Free
 }
]
