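# Animation Blueprint for the Ensemble Grey male character's face accessory part, adding physics motion on top of the pose copied from its attached parent mesh.
# A RigidBody node simulates the part in root-bone space, with gravity supplied each update by a deprecated gravity-override state function.
# The update event also switches the component velocity and acceleration scale on skydiving state.
# While skydiving, an additive pitch rotation is applied to dyn_hairTail_01 after simulation.
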AnimBlueprint /BRCosmetics/Characters/Player/Male/Medium/Bodies/M_MED_Ensemble_Grey/Meshes/Parts/M_MED_Ensemble_Grey_FaceAcc_AnimBP
target_skeleton=&/M_MED_Ensemble_Grey_FaceAcc_Skeleton
graphs [
 AnimGraph {nodes [
   n1 = Get skydive Acc Scale() -> (skydive Acc Scale:Vector)
   n2 = RigidBody!(
    ComponentPose=Local To Component(LocalPose=Copy Pose From Mesh() {bUseAttachedParent=true})
    Alpha
    OverrideWorldGravity:Vector=Get Gravity Override()
    ExternalForce
    ComponentLinearAccScale:Vector=@n1
    ComponentLinearVelScale:Vector=@n1
   ) -> (Pose) {
    bClampLinearTranslationLimitToRefPose=true
    bOverrideWorldGravity=true
    overridePhysicsAsset=&/M_MED_Ensemble_Grey_FaceAcc_Physics
    simulationSpace=BaseBoneSpace
    alphaBoolBlend {blendOption=Linear}
    baseBoneRef {boneName=root}
    componentAppliedLinearAccClamp=<3000 3000 3000>
    componentLinearAccScale=<0.01 0.01 0.01>
    componentLinearVelScale=<0.01 0.01 0.01>
    overrideWorldGravity {z=-980}
   }
   n3 = "Transform (Modify) Bone - Bone: dyn_hairTail_01"!(ComponentPose=@n2, bAlphaBoolEnabled:bool=Get bIsSkyDiving(), Translation="0,0,0", Rotation="20,0,0", Scale="1,1,1") -> (Pose) {
    alphaInputType=Bool
    rotationMode=BMM_Additive
    rotationSpace=BCS_BoneSpace
    alphaBoolBlend {blendOption=Linear}
    boneToModify {boneName=dyn_hairTail_01}
    rotation {pitch=20}
   }
   Output Pose(Result=Component To Local(ComponentPose=@n3)) {layerGroup=None}
  ]}
 EventGraph {nodes [
   n1 = Event BlueprintUpdateAnimation()
   > n2 = Sequence!(execute) -> (then_0, then_1, then_2:Exec)
   > Set anim Instance(anim Instance:Anim Instance Object Reference) -> (Output_Get:Anim Instance Object Reference)
   n3 = Gravity Override State Function DEPRECATED(
    execute=@n2.then_1
    CharacterPartAnimInstance=Self-Reference()
    GameTimeInSeconds=@n1.DeltaTimeX
    GravityOverrideParams=Get Gravity Override Params()
   )
   > Set Gravity Override(Gravity Override=@n3)
   n4 = Branch(execute=@n2.then_2, Condition=Get bIsSkyDiving())
   > Set skydive Acc Scale(skydive Acc Scale:Vector="0.001,0.001,0.001") -> (Output_Get:Vector)
   Set skydive Acc Scale(execute=@n4.else, skydive Acc Scale="0.1,0.1,0.6")
   Set body Mesh(body Mesh:Skeletal Mesh Component Object Reference=GetPartSkeletalMeshComponent(self=Self-Reference())) -> (Output_Get:Skeletal Mesh Component Object Reference)
  ]}
]
variables [
 anim Instance {type=AnimInstance, default=null}
 body Mesh {type=SkeletalMeshComponent}
 Gravity Override {type=Vector}
 Gravity Override Params {type=GravityOverrideParamsStruct, default {skydive_FloatingDown {gravityOverride=<0 0 1500>}}}
 skydive Acc Scale {type=Vector}
]
